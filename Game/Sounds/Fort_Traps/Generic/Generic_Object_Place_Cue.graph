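# Sound cue for generic trap object placement, routed to the Fort_Traps_Placement sound class.
# A single wave player feeds the output directly with a placed sound borrowed from the Dart Trap wall set, with no randomization or modulation.
# It uses the shared Fort_Traps_Att attenuation asset and its own concurrency override, which blocks new instances once the limit is reached instead of stopping playing ones.
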SoundCue /Game/Sounds/Fort_Traps/Generic/Generic_Object_Place_Cue
root=WavePlayer_41
nodes [{id=WavePlayer_41, properties {soundWaveAssetPtr=/Game/Sounds/Fort_Traps/Dart_Trap/WallTrap_Darts_Placed_002}}]
properties {
 attenuationSettings=&Fort_Traps_Att
 bOverrideConcurrency=true
 maxDistance=6000
 soundClassObject=/Game/Sounds/FortSoundClassesAndModes/Fort_Traps_Placement
 volumeMultiplier=0.9
 attenuationOverrides {falloffDistance=2400, attenuationShapeExtents {x=600}}
 concurrencyOverrides {resolutionRule=PreventNew}
}
